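# Niagara swing trail for the Chrome Siren pickaxe, built from one looping "Tiles" emitter that spawns anim-trail ribbon particles between two bone or socket indices.
# System-level fade values and a time-of-day color modulate each short-lived particle, while ribbon width tapers over life.
# Separate Basic, Blocks and Lines red-tinted color curves are written alongside a ribbon UV distance driven by particle ID and emitter age.
# Four ribbon renderers draw the layers.
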NiagaraSystem /BRCosmetics/Weapons/FORT_Melee/Pickaxe_ChromeSiren/FX/NS_Pickaxe_ChromeSiren_Trail
dependencies {
 usedDataInterfaces [&NiagaraDataInterfaceCurve, &NiagaraDataInterfaceColorCurve, &NiagaraDataInterfaceSocketReader]
 usedDynamicInputs [
  &MergeHelper_FloatSwitch
  &ClampFloat
  &RemapRange
  &Subtract_Float
  &FNTimeOfDayColor
 ]
 usedModules [
  &self:SystemSpawnScript.SystemScriptSource.SystemScriptGraph.NiagaraNodeAssignment_1.SetVariables_#1
  &self:SystemSpawnScript.SystemScriptSource.SystemScriptGraph.NiagaraNodeAssignment_3.SetVariables_#2
  &SystemState
  &self:SystemSpawnScript.SystemScriptSource.SystemScriptGraph.NiagaraNodeAssignment_4.SetVariables_#3
  &self:SystemSpawnScript.SystemScriptSource.SystemScriptGraph.NiagaraNodeAssignment_5.SetVariables_#4
 ]
}
emitters {Tiles {
  emitterSpawnScript [&FN_RibbonIDOffsetUtility()]
  emitterUpdateScript [
   &EmitterState(Life Cycle Mode=::NewEnumerator1, Loop Behavior=::NewEnumerator1, Loop Delay=0.05, Loop Duration=1, UseLoopDelay=true)
   &FN_SpawnAnimTrail()
  ]
  particleSpawnScript [
   &FN_InitializeAnimTrailBoneIndecies(First Bone or Socket Index=1, Second Bone or Socket Index=0)
   &FN_InitializeAnimTrail(
    Color=&MakeLinearColorFromVectorAndFloat(
     "Vector (RGB)"=&MakeVectorFromLinearColorRGB(Color=<rgba 1 1 1 1>)
     "Float (Alpha)"=&Multiply_Float(A=1, B=&Multiply_Float(A=$System.FadeSpawn, B=$System.FadeSpawnOut))
    )
    Lifetime=0.25
   ) {module=FN_InitializeAnimTrail001}
  ]
  particleUpdateScript [
   &FN_UpdateAnimTrail()
   &ScaleColor(
    Scale Alpha=&FloatFromCurve(
     FloatCurve=DataInterface() {propertyValues {Curve {
        defaultValue=1
        keys [{interpMode=RCIM_Cubic, tangentMode=RCTM_User, time=0.5154, value=1}, {interpMode=RCIM_Cubic, tangentMode=RCTM_User, time=1, value=1}]
       }}}
     Scale Curve=$System.FadeSystem
    )
    ScaleA=true
    ScaleRGB=true
   )
   &ScaleColor(Color Scale:LinearColor=$System.TODColor, Scale Mode=3) {module=ScaleColor001}
   &ScaleRibbonWidth(Ribbon Width Scale=&FloatFromCurve(FloatCurve=DataInterface() {propertyValues {Curve {keys [<at 0 1>, {arriveTangent=0.0896, time=1, value=0.8}]}}}))
   /SetVariables(Particles.LineParam0:Vector4f=<1 1 0 0>) {isSetParameters=true}
   /SetVariables(
    Particles.ColorBasic:LinearColor=&MultiplyLinearColor(
     Linear Color=$Particles.Color
     Scale Factor=&MakeVector4_FromLinearColor(LinearColor=&ColorFromCurve(ColorCurve=DataInterface() {propertyValues {
         AlphaCurve {keys [<at 0 0.85 RCIM_Cubic>,<at 1 0 RCIM_Cubic>]}
         BlueCurve {keys [<at 0 0.3525>,<at 0.548 0.1006 RCIM_Cubic>]}
         GreenCurve {keys [<at 0 0.2957>,<at 0.548 0.0696 RCIM_Cubic>]}
         RedCurve {keys [<at 0 2>,<at 0.548 1 RCIM_Cubic>]}
        }}))
    )
    Particles.ColorBlocks:LinearColor=&MultiplyLinearColor(
     Linear Color=$Particles.Color
     Scale Factor=&MakeVector4_FromLinearColor(LinearColor=&ColorFromCurve(ColorCurve=DataInterface() {propertyValues {
         AlphaCurve {keys [<at 0.1729 0.647 RCIM_Cubic>,<at 0.4573 0.0275>,<at 1 0 RCIM_Cubic>]}
         BlueCurve {keys [<at 0.1116 0.873>,<at 0.3583 0.1006 RCIM_Cubic>]}
         GreenCurve {keys [<at 0.1116 0.7652>,<at 0.3583 0.0696 RCIM_Cubic>]}
         RedCurve {keys [<at 0.1116 4>,<at 0.3583 1 RCIM_Cubic>]}
        }}))
    )
    Particles.ColorLines:LinearColor=&MultiplyLinearColor(
     Linear Color=$Particles.Color
     Scale Factor=&MakeVector4_FromLinearColor(LinearColor=&MakeLinearColorFromVectorAndFloat(
       "Vector (RGB)"=&MakeVectorFromLinearColorRGB(Color=<rgba 3 0.6753 0.66 1>)
       "Float (Alpha)"=&FloatFromCurve(FloatCurve=DataInterface() {propertyValues {Curve {keys [
            {arriveTangent=-1.1878, interpMode=RCIM_Cubic, time=0, value=5}
            {arriveTangent=-1.9684, interpMode=RCIM_Cubic, leaveTangent=-1.9684, tangentMode=RCTM_User, time=0.3966, value=1}
            {arriveTangent=-0.123, interpMode=RCIM_Cubic, leaveTangent=-0.123, tangentMode=RCTM_User, time=1, value=0}
           ]}}})
      ))
    )
   ) {isSetParameters=true}
   /SetVariables(Particles.RibbonUVDistance:NiagaraFloat=&Add_Float(
     A=&Multiply_FloatByInt(Float=5, Integer=&ReturnParticleIDAttributes())
     B=&Multiply_Float(A=-50.4266, B=$Emitter.Age)
    )) {isSetParameters=true}
  ]
  rendererClasses [&NiagaraRibbonRendererProperties]
  renderers [
   {rendererClass=&NiagaraRibbonRendererProperties}
   {rendererClass=&NiagaraRibbonRendererProperties, rendererIndex=1}
   {rendererClass=&NiagaraRibbonRendererProperties, rendererIndex=2}
   {rendererClass=&NiagaraRibbonRendererProperties, rendererIndex=3}
  ]
 }}
summary {
 systemName=NS_Pickaxe_ChromeSiren_Trail
 emitters [{emitterName=Tiles, rendererClasses [&NiagaraRibbonRendererProperties]}]
 userVariables [
  User.RibbonIDOffset {type=NiagaraInt32}
  User.SwingSectionIndex {type=NiagaraInt32}
 ]
}
