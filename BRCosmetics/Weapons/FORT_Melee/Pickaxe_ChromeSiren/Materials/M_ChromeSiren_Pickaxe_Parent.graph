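# Parent material for the ChromeSiren pickaxe, inserting custom glow and surface tweaks between the shared 2023 character PreFX and PostFX functions.
# Tech Art Mask red and blue channels define blade and groove masks, summed into a glass mask.
# Emission blends inner glow and powered highlights, scaled by Glow Enabled, while roughness and specular get masked adjustments.
# A sweeping gradient comes from local-position XY, panned over time in a vertex interpolator.
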Material /BRCosmetics/Weapons/FORT_Melee/Pickaxe_ChromeSiren/Materials/M_ChromeSiren_Pickaxe_Parent
graph {
 properties {
  bAutomaticallySetUsageInEditor=false
  bHasPixelAnimation=true
  bUseMaterialAttributes=true
  bUsedWithMeshDeformer=true
  bUsedWithMorphTargets=true
  bUsedWithSkeletalMesh=true
  pixelDepthOffsetMode=PDOM_Legacy
  previewMesh=/BRCosmetics/Weapons/FORT_Melee/Pickaxe_ChromeSiren/Meshes/ChromeSiren_Axe
  shadingModel=MSM_FromMaterialExpression
  usedShadingModels=Default Lit | Subsurface | Clear Coat | Subsurface Profile | Cloth
 }
 nodes [
  Constant()
  n1 = MaterialFunctionCall() {materialFunction=&MF_CharacterParent_2023_PreFX}
  n2 = GetMaterialAttributes(None=@n1.Result) {attributeGetTypes [
    #1
    #2
    #3
    #4
   ]}
  n3 = SetMaterialAttributes!(
   MaterialAttributes=@n2.MaterialAttributes
   Emissive Color=NamedRerouteUsage()
   Base Color
   Roughness=NamedRerouteUsage()
   Specular=NamedRerouteUsage()
  ) {attributeSetTypes [
    #1
    #2
    #3
    #4
   ]}
  MaterialFunctionCall(Material=@n3, M=@n1.M, Pre Skinned Local Position=@n1.M, Distance Blend=@n1.M, Fresnel=@n1.M) {materialFunction=&MF_CharacterParent_2023_PostFX}
  n4 = NamedRerouteDeclaration(None=ComponentMask(None=@n1.Tech Art Mask) {r=true}) {name=BladeMask, nodeColor=<rgba 1 0 0.8438 1>}
  n5 = NamedRerouteDeclaration(None=ComponentMask(None=@n1.Tech Art Mask) {b=true}) {name=GrooveMask, nodeColor=<rgba 0 1 0.8359 1>}
  NamedRerouteDeclaration(None=Saturate(None=Add(A=@n4, B=@n5))) {name=GlassMask, nodeColor=<rgba 0.1016 0 1 1>}
  n6 = NamedRerouteUsage()
  n7 = NamedRerouteUsage()
  n8 = Power(Base=@n7) {constExponent=0.9}
  n9 = Multiply(
   A=Multiply(A=NamedRerouteUsage(), B=Multiply(A=@n6, B=@n8))
   B=ScalarParameter() {defaultValue=1, group=03 - Pattern, parameterName=Inner Glow Intensity, sortPriority=6}
  ) {constB=0.5}
  n10 = LinearInterpolate(
   A=NamedRerouteUsage()
   B=@n7
   Alpha=ScalarParameter() {defaultValue=1, group=03 - Pattern, parameterName=Highlight Intensity, sortPriority=6}
  ) {constAlpha=0.25}
  n11 = ScalarParameter() {defaultValue=4, group=03 - Pattern, parameterName=Highlight Power, sortPriority=6}
  n12 = Multiply(
   A=MaterialFunctionCall!(Base=NamedRerouteUsage(), Add=Power(Base=@n10, Exp=@n11) {constExponent=3}) {materialFunction=/Engine/Functions/Engine_MaterialFunctions02/Math/MultiplyAdd}.Result
   B=NamedRerouteUsage()
  )
  n13 = NamedRerouteUsage()
  NamedRerouteDeclaration(None=LinearInterpolate(
    B=LinearInterpolate(A=@n9, B=@n12, Alpha=@n13) {constB=0}
    Alpha=ScalarParameter() {defaultValue=1, group=03 - Pattern, parameterName=Glow Enabled, sliderMax=1, sortPriority=1}
   )) {name=EmissionOUT, nodeColor=<rgba 1 0 0.1406 1>}
  n14 = Panner(Coordinate=Multiply(A=LocalPosition().XY) {constB=0.01}, Speed=Constant2Vector() {r=0.25}) {constCoordinate=1, speedX=0.25}
  n15 = Frac(None=MaterialFunctionCall(UV Channel=VertexInterpolator(VS=@n14).PS) {materialFunction=&LinearGradient}.UGradient)
  n16 = Multiply(A=Multiply(A=@n15, B=OneMinus(None=@n15))) {constB=4}
  NamedRerouteDeclaration(None=OneMinus(None=Saturate(None=@n16))) {name=SweepingGradient, nodeColor=<rgba 0 0.4609 1 1>}
  NamedRerouteDeclaration(None=@n2.Base Color) {name=BaseColorIN, nodeColor=<rgba 1 0 0.0469 1>}
  NamedRerouteDeclaration(None=@n2.Emissive Color) {name=EmisionIN, nodeColor=<rgba 0.5703 0 1 1>}
  NamedRerouteDeclaration(None=@n2.Roughness) {name=RoughnessIN, nodeColor=<rgba 0.6172 1 0 1>}
  NamedRerouteDeclaration(None=@n2.Specular) {name=SpecularIN, nodeColor=<rgba 1 0 0.9375 1>}
  n17 = NamedRerouteUsage()
  n18 = Constant() {r=0.07}
  NamedRerouteDeclaration(None=LinearInterpolate(A=@n17, B=Add(A=@n17, B=@n18), Alpha=NamedRerouteUsage())) {name=RoughnessOUT, nodeColor=<rgba 0.6172 1 0 1>}
  n19 = NamedRerouteUsage()
  n20 = Constant4Vector() {constant=<rgba 0.7 0.95 0 1>}
  NamedRerouteDeclaration(None=LinearInterpolate(
    A=@n19
    B=MaterialFunctionCall(Input=@n19, Input Low=@n20.R, Input High=@n20.R, Target Low=@n20.R, Target High=@n20.R) {materialFunction=&RemapValueRange}.Result
    Alpha=NamedRerouteUsage()
   )) {name=SpecularOUT, nodeColor=<rgba 1 0 0.9375 1>}
 ]
}
